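# Physics asset for the Black Widow Jacket backpack cosmetic, giving capsule bodies to its cloth, straps and zipper pulls for secondary motion.
# Four chains hang off the pack body: six cloth segments, left and right three-segment zippers, and two strap segments.
# Each segment is constrained to its parent, with wide angular limits on cloth, tighter swing on zippers, and zero swing limits on straps, leaving them twist.
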PhysicsAsset /BRCosmetics/Accessories/FORT_Backpacks/Backpack_F_MED_Black_Widow_Jacket/Meshes/F_MED_Black_Widow_Jacket_Pack_Physics
bodies {
 dyn_cloth_01 [{shapeType=Capsule, length=4.755, radius=1.255, rotation=<rot 0 -90 0>}]
 dyn_cloth_02 [{shapeType=Capsule, length=4.755, radius=1.255, center=<0 0.0125 -2.453>, rotation=<rot 0.0041 -90 0>}]
 dyn_cloth_03 [{shapeType=Capsule, length=4.505, radius=1.505, center=<0 0.016 -3.2907>, rotation=<rot -0.0859 -90 0>}]
 dyn_cloth_04 [{shapeType=Capsule, length=4.505, radius=1.505, center=<0.0015 0.019 -2.9167>, rotation=<rot -0.0637 -89.9998 1.3715>}]
 dyn_cloth_05 [{shapeType=Capsule, length=4.505, radius=1.505, center=<0 0.022 -2.4655>, rotation=<rot 0.0222 -90 0>}]
 dyn_cloth_06 [{shapeType=Capsule, length=4.505, radius=1.505, center=<0 0.0297 -3.0861>, rotation=<rot -0.3131 -90 0>}]
 dyn_straps_01 [{shapeType=Capsule, length=4.0365, radius=1.5436, center=<0 -0.0024 1.0473>, rotation=<rot 0 -90 -90>}]
 dyn_straps_02 [{shapeType=Capsule, length=3.5677, radius=1.7712, center=<0 -1.6744 1.0568>, rotation=<rot 0 -89.9986 -89.9997>}]
 dyn_zippers_l_01 [{shapeType=Capsule, length=1.2496, radius=0.5535, center=<-0.0017 -0.0015 -1.052>, rotation=<rot 0.6509 -90 -0.5887>}]
 dyn_zippers_l_02 [{shapeType=Capsule, length=0.8799, radius=0.5508, center=<-0.0053 0.0114 -0.6285>, rotation=<rot -1.4362 -90 -0.2098>}]
 dyn_zippers_l_03 [{shapeType=Capsule, length=5.5092, radius=1.9959, center=<-0.0153 0.3813 -3.8502>, rotation=<rot 2.9978 -90 -0.0228>}]
 dyn_zippers_r_01 [{shapeType=Capsule, length=1.2813, radius=0.5529, center=<0.0019 -0.0057 -0.9938>, rotation=<rot 0.2321 -90 0.5076>}]
 dyn_zippers_r_02 [{shapeType=Capsule, length=0.9984, radius=0.5533, center=<0.0039 -0.0068 -0.6501>, rotation=<rot 0.1039 -90 -0.4926>}]
 dyn_zippers_r_03 [{shapeType=Capsule, length=5.5168, radius=1.9928, center=<0.0162 0.3728 -3.7575>, rotation=<rot 3.3176 -90 0.0049>}]
 pack [{shapeType=Capsule, length=13.0917, radius=5.0213, center=<0.0002 -3.7081 -2.1032>, rotation=<rot 2.1204 -89.9999 0>}]
}
constraints [
 {bone1Name=dyn_cloth_01, bone2Name=pack, swing1LimitDegrees=45, swing2LimitDegrees=45, twistLimitDegrees=45}
 {bone1Name=dyn_cloth_02, bone2Name=dyn_cloth_01, swing1LimitDegrees=45, swing2LimitDegrees=45, twistLimitDegrees=45}
 {bone1Name=dyn_cloth_03, bone2Name=dyn_cloth_02, swing1LimitDegrees=45, swing2LimitDegrees=45, twistLimitDegrees=45}
 {bone1Name=dyn_cloth_04, bone2Name=dyn_cloth_03, swing1LimitDegrees=45, swing2LimitDegrees=45, twistLimitDegrees=45}
 {bone1Name=dyn_cloth_05, bone2Name=dyn_cloth_04, swing1LimitDegrees=45, swing2LimitDegrees=45, twistLimitDegrees=45}
 {bone1Name=dyn_cloth_06, bone2Name=dyn_cloth_05, swing1LimitDegrees=45, swing2LimitDegrees=45, twistLimitDegrees=45}
 {bone1Name=dyn_zippers_l_01, bone2Name=pack, swing1LimitDegrees=10, swing2LimitDegrees=10, twistLimitDegrees=45}
 {bone1Name=dyn_zippers_l_02, bone2Name=dyn_zippers_l_01, swing1LimitDegrees=10, swing2LimitDegrees=10, twistLimitDegrees=45}
 {bone1Name=dyn_zippers_l_03, bone2Name=dyn_zippers_l_02, swing1LimitDegrees=10, swing2LimitDegrees=10, twistLimitDegrees=45}
 {bone1Name=dyn_zippers_r_01, bone2Name=pack, swing1LimitDegrees=10, swing2LimitDegrees=10, twistLimitDegrees=45}
 {bone1Name=dyn_zippers_r_02, bone2Name=dyn_zippers_r_01, swing1LimitDegrees=10, swing2LimitDegrees=10, twistLimitDegrees=45}
 {bone1Name=dyn_zippers_r_03, bone2Name=dyn_zippers_r_02, swing1LimitDegrees=10, swing2LimitDegrees=10, twistLimitDegrees=45}
 {bone1Name=dyn_straps_01, bone2Name=pack, swing1LimitDegrees=0, swing2LimitDegrees=0, twistLimitDegrees=5}
 {bone1Name=dyn_straps_02, bone2Name=dyn_straps_01, swing1LimitDegrees=0, swing2LimitDegrees=0, twistLimitDegrees=45}
]
