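# Physics asset for the Black Widow Jacket backpack cosmetic, giving its pack, dangling cloth and straps capsule bodies for bone-driven physics.
# A six-link cloth chain and a two-link strap chain both hang from the large pack capsule, each bone constrained to the one before it.
# Cloth joints carry uniform, moderate swing and twist limits, while strap joints are set much tighter, with zero swing and slight twist.
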PhysicsAsset /BRCosmetics/Accessories/FORT_Backpacks/Backpack_F_MED_Black_Widow_Jacket/Meshes/F_MED_Black_Widow_Jacket_Pack_cloth_Physics
bodies {
 dyn_cloth_01 [{shapeType=Capsule, length=4.755, radius=1.255, rotation=<rot 0 -90 0>}]
 dyn_cloth_02 [{shapeType=Capsule, length=4.755, radius=1.255, center=<0 0.0125 -2.453>, rotation=<rot 0.0041 -90 0>}]
 dyn_cloth_03 [{shapeType=Capsule, length=4.755, radius=1.255, center=<0 0.016 -3.2907>, rotation=<rot -0.0859 -90 0>}]
 dyn_cloth_04 [{shapeType=Capsule, length=4.755, radius=1.255, center=<0.0015 0.019 -2.9167>, rotation=<rot -0.0637 -89.9998 1.3715>}]
 dyn_cloth_05 [{shapeType=Capsule, length=4.755, radius=1.255, center=<0 0.022 -2.4655>, rotation=<rot 0.0222 -90 0>}]
 dyn_cloth_06 [{shapeType=Capsule, length=4.755, radius=1.255, center=<0 0.0297 -3.0861>, rotation=<rot -0.3131 -90 0>}]
 dyn_straps_01 [{shapeType=Capsule, length=4.0365, radius=1.5436, center=<0 -0.0024 1.0473>, rotation=<rot 0 -90 -90>}]
 dyn_straps_02 [{shapeType=Capsule, length=3.5677, radius=1.7712, center=<0 -1.6744 1.0568>, rotation=<rot 0 -89.9986 -89.9997>}]
 pack [{shapeType=Capsule, length=16.5917, radius=10.5213, center=<0.0002 1.3284 -7.362>, rotation=<rot -1.3191 -89.9999 0>}]
}
constraints [
 {bone1Name=dyn_cloth_01, bone2Name=pack, swing1LimitDegrees=45, swing2LimitDegrees=45, twistLimitDegrees=45}
 {bone1Name=dyn_cloth_02, bone2Name=dyn_cloth_01, swing1LimitDegrees=45, swing2LimitDegrees=45, twistLimitDegrees=45}
 {bone1Name=dyn_cloth_03, bone2Name=dyn_cloth_02, swing1LimitDegrees=45, swing2LimitDegrees=45, twistLimitDegrees=45}
 {bone1Name=dyn_cloth_04, bone2Name=dyn_cloth_03, swing1LimitDegrees=45, swing2LimitDegrees=45, twistLimitDegrees=45}
 {bone1Name=dyn_cloth_05, bone2Name=dyn_cloth_04, swing1LimitDegrees=45, swing2LimitDegrees=45, twistLimitDegrees=45}
 {bone1Name=dyn_cloth_06, bone2Name=dyn_cloth_05, swing1LimitDegrees=45, swing2LimitDegrees=45, twistLimitDegrees=45}
 {bone1Name=dyn_straps_01, bone2Name=pack, swing1LimitDegrees=0, swing2LimitDegrees=0, twistLimitDegrees=10}
 {bone1Name=dyn_straps_02, bone2Name=dyn_straps_01, swing1LimitDegrees=0, swing2LimitDegrees=0, twistLimitDegrees=10}
]
